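# Parent material for the Hatchback bass instrument cosmetic, built on material attributes for skeletal meshes, with cloth and mesh deformer support.
# MF_Hatchback_ComicShading, given custom UVs from the shared MF_EyeUVs function, builds the surface that feeds MF_CharacterParent_2023_PostFX.
# A collapsed subgraph supplies that PostFX function with a constant M, distance blend, Fresnel and pre-skinned local position.
# A shader stage switch routes that position through a vertex interpolator for the pixel shader.
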Material /SparksCosmetics/Instrument/Bass/DriedSilk/Materials/M_Hatchback_Parent
graph {
 properties {
  bAutomaticallySetUsageInEditor=false
  bUseMaterialAttributes=true
  bUsedWithClothing=true
  bUsedWithMeshDeformer=true
  bUsedWithSkeletalMesh=true
  bUsedWithStaticLighting=true
  pixelDepthOffsetMode=PDOM_Legacy
  shadingModel=MSM_FromMaterialExpression
  usedShadingModels=Default Lit
 }
 nodes [
  Composite() {subgraphName=Construct Base Inputs}
  n1 = Constant()
  MaterialExpressionReroute_4 = Reroute(None=@n1)
  n2 = PreSkinnedPosition()
  n3 = ShaderStageSwitch(PixelShader=VertexInterpolator(VS=@n2).PS, VertexShader=@n2)
  MaterialExpressionReroute_5 = Reroute(None=@n3)
  n4 = MaterialFunctionCall!(Material) {materialFunction=/Game/Packages/Fortress_SharedMaterials/Base_Material_Functions/CharacterBrightnessandScale_Balance}
  MaterialExpressionReroute_6 = Reroute(None=@n4.Distance Blend)
  n5 = Fresnel() {baseReflectFraction=0, exponent=1}
  MaterialExpressionReroute_7 = Reroute(None=@n5)
  MaterialFunctionCall(
   Material=MaterialFunctionCall!(CustomUVs=MaterialFunctionCall() {materialFunction=/Game/Characters/Shared/FNEye/Materials/MF_EyeUVs}.EyeUVs) {materialFunction=&/MF_Hatchback_ComicShading}.Result
   M=@MaterialExpressionReroute_4.M
   Pre Skinned Local Position=@MaterialExpressionReroute_5.Pre Skinned Local Position
   Distance Blend=@MaterialExpressionReroute_6.Distance Blend
   Fresnel=@MaterialExpressionReroute_7.Fresnel
  ) {materialFunction=&MF_CharacterParent_2023_PostFX}
  PinBase()
  PinBase!(M=@n1, Pre Skinned Local Position=@n3, Distance Blend=@n4.Distance Blend, Fresnel=@n5) {reroutePins [
    M {expression=&self:MaterialExpressionReroute_4}
    Pre Skinned Local Position {expression=&self:MaterialExpressionReroute_5}
    Distance Blend {expression=&self:MaterialExpressionReroute_6}
    Fresnel {expression=&self:MaterialExpressionReroute_7}
   ]}
 ]
}
